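; BP_DistantFog is a construction script for a distant fog environment Blueprint that dresses a fog mesh with configurable tint and opacity.
; It assigns the distant fog material to the fog mesh.
; It then loops over the root's direct children, casts each to a static mesh component, and pushes the fog color and opacity variables into the FogTint and FogOpacity material parameters.
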(fn ConstructionScript
  (Rendering|Material|SetMaterial (Variables|Default|GetFogMesh) 0 (Variables|Default|GetDistantFogMaterial))
  (for _array_element (Components|GetChildrenComponents (Variables|Default|GetDefaultSceneRoot) false)
    (bind _asstatic_mesh_component (Utilities|Casting|CastToStaticMeshComponent _array_element))
    (Rendering|Material|SetVectorParameterValueonMaterials _asstatic_mesh_component "FogTint" (Math|Conversions|ToVector(LinearColor) (Variables|Default|GetDistanceFog%20Color)))
    (Rendering|Material|SetScalarParameterValueonMaterials _asstatic_mesh_component "FogOpacity" (Variables|Default|GetDistance%20Fog%20Opacity))))
